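; Construction-script Blueprint for the strong variant of the Dark Nexus tower light source prop.
; After the parent construction, it reads the building texture data and applies override materials: the second entry goes to the TowerLight_B mesh when exactly two exist, otherwise the first goes to the main static mesh.
; When Use TOD Component is set, it runs the time-of-day lighting component's construct with a test day phase.
; Its event graph events are empty.
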(event EventBeginPlay)

(event Collision|EventActorBeginOverlap
  (params (OtherActor :type Actor%20Object%20Reference)))

(event EventTick
  (params (DeltaSeconds :type Float%20%28single-precision%29)))

(fn ConstructionScript
  (|CallParentFunction)
  (bind _outtexturedataarray (Building|GetBuildingTextureData))
  (if (== (Utilities|Array|Length _outtexturedataarray) 2)
    (Rendering|Material|SetMaterial (Variables|Default|GetSM_LightSource_TowerLight_B) 0 (Variables|/Script/FortniteGame.BuildingTextureData|GetOverrideMaterial (Utilities|Array|Get(acopy) _outtexturedataarray 1)))
    (Rendering|Material|SetMaterial (Variables|Debug|GetStaticMeshComponent) 0 (Variables|/Script/FortniteGame.BuildingTextureData|GetOverrideMaterial (Utilities|Array|Get(acopy) _outtexturedataarray))))
  (if (Variables|Default|GetUse%20TOD%20Component)
    (Class|TODLightingComponent|RunConstruct (Variables|Default|GetTODLightingComponent) (Variables|Default|GetTest%20Day%20Phase_0))))
